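# Editor utility shelf for tech artists, holding saved, horizontally scrolling buttons that launch editor widgets, console commands, Python or actor spawns.
# Shift-click instead locates a button's widget asset in the Content Browser, and actors spawn into the game world if valid, otherwise the editor.
# A popup tab adds buttons, delete mode exposes removal, and SaveGame persists the list.
# Tick polls a flag to import Python-supplied button arrays, skipping duplicate labels.
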EditorUtilityWidgetBlueprint /Game/Blueprints/Blutilities/ForTechArt/EditorWidgetShelf/EditorWidgetShelf_EWidget
graphs [
 EventGraph {nodes [
   "On Clicked (TA_Generic01_Button)"!() -> (OutputDelegate, then, widget:User Widget Object Reference, event:Pointer Event Structure) {is=title K2Node_ComponentBoundEvent}
   > handleAddPopupCanceled!(execute, self, shouldCleanUpPopup:bool=true) -> (then)
   "On Clicked (btnAddNewButton)"!() -> (OutputDelegate, then, widget:User Widget Object Reference, event:Pointer Event Structure) {is=title K2Node_ComponentBoundEvent}
   > openAddButtonPopup()
   n1 = "On Item Left Clicked (TA_Generic01_List)"!() -> (OutputDelegate, then, isShiftDown:bool, isCtrlDown:bool, data:Object Reference, item:User Widget Object Reference) {is=title K2Node_ComponentBoundEvent}
   > handleToolButtonClicked!(execute, self, data:Object Reference=@n1.data, isShiftDown:bool=@n1.isShiftDown) -> (then)
   "On Clicked (btnDeleteMode)"!() -> (OutputDelegate, then, widget:User Widget Object Reference, event:Pointer Event Structure) {is=title K2Node_ComponentBoundEvent}
   > setDeleteMode!(execute, self, on:bool=NOT Boolean(A=Get deleteModeOn() -> (deleteModeOn:bool))) -> (then)
   Event Tick()
   > Branch(Condition=Get pythonWantsUpdate() -> (pythonWantsUpdate:bool))
   > Set pythonWantsUpdate(pythonWantsUpdate:bool=false) -> (Output_Get:bool)
   > addPythonButtons()
   Event Destruct()
   > handleAddPopupCanceled(shouldCleanUpPopup=true)
  ]}
 SaveGame_PostLoad {nodes [
   n1 = SaveGame_PostLoad!() -> (then, saveGameExisted:bool, saveGame_Object:Save Game Object Reference) {is=title K2Node_FunctionEntry}
   > Branch(Condition=@n1)
   > n2 = Cast To EWS_SaveGame!(execute, Object=@n1.saveGame_Object) -> (then, CastFailed, AsEWS Save Game:EWS Save Game Object Reference) {is=title K2Node_DynamicCast}
   > n3 = For Each Loop(Array:Array of EWS Button Data Struct Structures=Get buttonData!(self:EWS Save Game Object Reference=@n2) -> (buttonData:Array of EWS Button Data Struct Structures)) -> (Array Element:EWS Button Data Struct Structure)
   refreshList(execute=@n3.Completed)
   addButtonDataToList!(execute=@n3.LoopBody, self, data:"EWS Button Data Struct Structure (by ref)"=@n3) -> (then, allow:bool)
   Return Node!(execute, " ":bool=false)
  ]}
 SaveGame_PrepareToSave {nodes [
   SaveGame_PrepareToSave!() -> (then, SaveGameObject:Save Game Object Reference) {is=title K2Node_FunctionEntry}
   > n1 = Sequence()
   > For Each Loop!(Exec, Array=Get buttons() -> (buttons:Array of EWS Button Data Struct Structures)) -> (
    LoopBody
    Array Element_label_2_#1:String
    Array Element_widgetPackageName_12_#2:String
    Array Element_color_7_#3:Linear Color Structure
    Array Element_consoleCommand_16_#4:String
    Array Element_python_17_#5:String
    Array Element_actorPackageName_22_#6:String
    Array Element_actorIsTransient_24_#7:bool=~
    Array Index
    Completed
   )
   n2 = Cast To EWS_SaveGame(execute=@n1.then_1, Object=Get saveGame_Object() -> (saveGame_Object:Save Game Object Reference))
   > Set buttonData!(execute, buttonData:Array of EWS Button Data Struct Structures=Get buttons(), self:EWS Save Game Object Reference=@n2) -> (then, Output_Get:Array of EWS Button Data Struct Structures)
   Return Node(" "=false)
  ]}
 addButtonDataToList {nodes [
   n1 = addButtonDataToList!() -> (then, data:"EWS Button Data Struct Structure (by ref)") {is=title K2Node_FunctionEntry}
   n2 = Break EWS Button Data Struct!(EWS_ButtonData_Struct:"EWS Button Data Struct Structure (by ref)"=@n1) -> (
    label_2_#1:String
    widgetPackageName_12_#2:String
    color_7_#3:Linear Color Structure
    consoleCommand_16_#4:String
    python_17_#5:String
    actorPackageName_22_#6:String
    actorIsTransient_24_#7:bool=~
   ) {is=title K2Node_BreakStruct}
   > n3 = Branch(Condition=Contains Item(TargetArray:Array of Strings=Get currentButtonsNames() -> (currentButtonsNames:Array of Strings), ItemToFind:"String (by ref)"=@n2))
   > Return Node!(execute, allow:bool=false)
   Add(
    execute=@n3.else
    TargetArray:Array of EWS Button Data Struct Structures=Get buttons()
    NewItem:"EWS Button Data Struct Structure (by ref)"=@n1
   ) {is=title K2Node_CallArrayFunction}
   > Add(TargetArray:Array of Strings=Get currentButtonsNames(), NewItem:"String (by ref)"=@n2)
   > Return Node(allow=true)
  ]}
 addNewButton {nodes [
   n1 = addNewButton!() -> (then, NewItem:"EWS Button Data Struct Structure (by ref)") {is=title K2Node_FunctionEntry}
   > Set l_newButtonData(l_newButtonData:EWS Button Data Struct Structure=@n1) -> (Output_Get:EWS Button Data Struct Structure)
   > n2 = Construct Object from Class(
    Class=/Game/Blueprints/Blutilities/ForTechArt/EditorWidgetShelf/Data/EWS_Button_ListItemData.EWS_Button_ListItemData_C
    self=Self-Reference()
   ) -> (ReturnValue:EWS Button List Item Data Object Reference)
   n3 = Get l_newButtonData() -> (l_newButtonData:EWS Button Data Struct Structure)
   > Set buttonData(buttonData:EWS Button Data Struct Structure=@n3, self:EWS Button List Item Data Object Reference=@n2) -> (Output_Get:EWS Button Data Struct Structure)
   > Set displayString!(
    execute
    displayString:String=Break EWS Button Data Struct(EWS_ButtonData_Struct=@n3) -> (actorIsTransient_24_#7=~)
    self:TA Generic List Item Data Object Reference=@n2
   ) -> (then, Output_Get:String)
   > addItem!(
    execute
    self:TA Generic 01 List Object Reference=Get TA_Generic01_List() -> (TA_Generic01_List:TA Generic 01 List Object Reference)
    dataObject:Object Reference=@n2
   ) -> (then, newItem:User Widget Object Reference, newData:Object Reference)
  ]}
 addPythonButtons {nodes [
   addPythonButtons() {is=title K2Node_FunctionEntry}
   > n1 = For Each Loop with Break!(Exec:Exec, Array:Array of Strings=Get pythonReturn_Labels() -> (pythonReturn_Labels:Array of Strings), Break:Exec) -> (LoopBody:Exec, Array Element:String, Array Index:int, Completed:Exec)
   Branch(execute=@n1.Completed, Condition=Get l_addedSomething() -> (l_addedSomething:bool))
   > refreshList()
   > SaveGame_Save(self:TA Base Editor Widget Object Reference)
   n2 = "Get (a ref)"(Array:Array of Vectors=Get pythonReturn_Colors() -> (pythonReturn_Colors:Array of Vectors), Dimension 1=@n1.Array Index) -> (Output:"Vector (by ref)")
   n3 = Get pythonReturn_WidgetPackageNames() -> (pythonReturn_WidgetPackageNames:Array of Strings)
   n4 = Get pythonReturn_ConsoleCommands() -> (pythonReturn_ConsoleCommands:Array of Strings)
   n5 = Get pythonReturn_Pythons() -> (pythonReturn_Pythons:Array of Strings)
   n6 = addButtonDataToList(
    execute=@n1.LoopBody
    data=Make EWS Button Data Struct!(
     label_2_#1:String=@n1
     widgetPackageName_12_#2:String="Get (a ref)"(Array:Array of Strings=@n3, Dimension 1=@n1.Array Index) -> (Output:"String (by ref)")
     color_7_#3:Linear Color Structure="To LinearColor (Vector)"!(InVec:Vector=@n2) -> (ReturnValue:Linear Color Structure)
     consoleCommand_16_#4:String="Get (a ref)"(Array=@n4, Dimension 1=@n1.Array Index)
     python_17_#5:String="Get (a ref)"(Array=@n5, Dimension 1=@n1.Array Index)
     actorPackageName_22_#6:String
     actorIsTransient_24_#7:bool=False
    ) -> (EWS_ButtonData_Struct:EWS Button Data Struct Structure) {is=title K2Node_MakeStruct}
   ) {is=title K2Node_CallFunction}
   > Branch(Condition=@n6)
   > Set l_addedSomething(l_addedSomething:bool=true) -> (Output_Get:bool)
  ]}
 deleteShelfButton {nodes [
   n1 = deleteShelfButton!() -> (then, button:User Widget Object Reference) {is=title K2Node_FunctionEntry}
   > n2 = Cast To EWS_Button_Widget!(execute, Object=@n1) -> (then, CastFailed, AsEWS Button Widget:EWS Button Widget Object Reference) {is=title K2Node_DynamicCast}
   > removeButton!(execute, self, button:EWS Button Widget Object Reference=@n2) -> (then)
   > Return Node!(execute, " ":bool=false)
  ]}
 handleAddPopupCanceled {nodes [
   n1 = handleAddPopupCanceled!() -> (then, shouldCleanUpPopup:bool=true) {is=title K2Node_FunctionEntry}
   > n2 = Sequence()
   > SetVisibility(self=Get borderNewButtonPopup() -> (borderNewButtonPopup:Border Object Reference), InVisibility=Collapsed)
   Branch(execute=@n2.then_1, Condition=@n1)
   > Is Valid(InputObject=Get currentAddButtonPopup() -> (currentAddButtonPopup:EWS Add New Popup EWidget Object Reference))
   n3 = Get currentAddButtonPopup()
   > n4 = Branch(Condition="Equal (Object)"(
     A=Get parentShelfWidget!(self:EWS Add New Popup EWidget Object Reference=@n3) -> (parentShelfWidget:Editor Widget Shelf EWidget Object Reference)
     B=Self-Reference()
    ) {is=title K2Node_CallFunction})
   > n5 = setNewParent!(
    execute
    self:EWS Add New Popup EWidget Object Reference=Get currentAddButtonPopup()
    newParentShelfWidget:Editor Widget Shelf EWidget Object Reference
   ) -> (then)
   Set currentAddButtonPopup(execute=[@n5.then, @n4.else], currentAddButtonPopup:EWS Add New Popup EWidget Object Reference) -> (Output_Get:EWS Add New Popup EWidget Object Reference)
  ]}
 handleAddPopupComplete {nodes [
   n1 = handleAddPopupComplete!() -> (
    then
    label:String
    widget:Object Reference
    color:Linear Color Structure
    consoleCommand:String
    Python:String
    actor:String
    actorIsTransient:bool=~
   ) {is=title K2Node_FunctionEntry}
   n2 = Make EWS Button Data Struct(
    label_2_#1=@n1
    widgetPackageName_12_#2=Split(SourceString=Get Object Path String(Object=@n1.widget), InStr=., SearchDir=FromEnd)
    color_7_#3=@n1.color
    consoleCommand_16_#4=@n1.consoleCommand
    python_17_#5=@n1.Python
    actorPackageName_22_#6=@n1.actor
    actorIsTransient_24_#7=@n1.actorIsTransient
   )
   > n3 = addButtonDataToList(data=@n2)
   > Branch(Condition=@n3)
   > addNewButton!(execute, self, NewItem:"EWS Button Data Struct Structure (by ref)"=@n2) -> (then) {is=title K2Node_CallFunction}
   > SetVisibility(self=Get borderNewButtonPopup(), InVisibility=Collapsed)
   > SaveGame_Save()
  ]}
 handleToolButtonClicked {nodes [
   handleToolButtonClicked!() -> (then, data:Object Reference, isShiftDown:bool=~) {is=title K2Node_FunctionEntry}
   > n1 = Branch(Condition=Get isShiftDown() -> (isShiftDown:bool))
   > n2 = Cast To EWS_Button_ListItemData!(execute, Object=Get data() -> (data:Object Reference)) -> (then, CastFailed, AsEWS Button List Item Data:EWS Button List Item Data Object Reference) {is=title K2Node_DynamicCast}
   > n3 = GetAssetsByPackageName!(
    execute
    self:Asset Registry Interface=GetAssetRegistry()
    PackageName:Name=String To Name(InString=Get!(self:EWS Button List Item Data Object Reference=@n2) -> (
      buttonData_label_2_#1:String
      buttonData_widgetPackageName_12_#2:String
      buttonData_color_7_#3:Linear Color Structure
      buttonData_consoleCommand_16_#4:String
      buttonData_python_17_#5:String
      buttonData_actorPackageName_22_#6:String
      buttonData_actorIsTransient_24_#7:bool=~
     ).buttonData_widgetPackageName_12_#2)
    bIncludeOnlyOnDiskAssets:bool=false
    bSkipARFilteredAssets:bool=true
   ) -> (then, OutAssetData:Array of Asset Data Structures, ReturnValue:bool)
   > Branch(Condition=Is Valid Index!(TargetArray:Array of Asset Data Structures=@n3, IndexToTest:int) -> (ReturnValue:bool))
   n4 = "Get (a copy)"!(Array:Array of Asset Data Structures=@n3, Dimension 1) -> (Output_PackageName:Name, Output_PackagePath:Name, Output_AssetName:Name, Output_AssetClassPath:Top Level Asset Path Structure)
   > SyncBrowserToObjects(AssetPaths=Make Array("[0]":String="To String (Name)"(InName=@n4)) -> (Array:Array of Strings))
   n5 = Cast To EWS_Button_ListItemData(execute=@n1.else, Object=Get data())
   > Set l_dataStruct(l_dataStruct:EWS Button Data Struct Structure=Get buttonData(self:EWS Button List Item Data Object Reference=@n5) -> (buttonData:EWS Button Data Struct Structure)) -> (Output_Get:EWS Button Data Struct Structure)
   > n6 = Sequence!(execute) -> (then_0, then_1, then_2:Exec, then_3:Exec)
   n7 = Get!() -> (
    l_dataStruct_label_2_#1:String
    l_dataStruct_widgetPackageName_12_#2:String
    l_dataStruct_color_7_#3:Linear Color Structure
    l_dataStruct_consoleCommand_16_#4:String
    l_dataStruct_python_17_#5:String
    l_dataStruct_actorPackageName_22_#6:String
    l_dataStruct_actorIsTransient_24_#7:bool=~
   )
   > Branch(Condition="Not Equal Exactly (String)"(A=@n7.l_dataStruct_widgetPackageName_12_#2))
   > n8 = GetAssetsByPackageName(
    self=GetAssetRegistry()
    PackageName=String To Name(InString=@n7.l_dataStruct_widgetPackageName_12_#2)
    bIncludeOnlyOnDiskAssets=false
    bSkipARFilteredAssets=true
   )
   > n9 = For Each Loop!(Exec, Array:Array of Asset Data Structures=@n8) -> (LoopBody, Array Element:Asset Data Structure, Array Index, Completed)
   Return Node!(execute=@n9.Completed)
   > n10 = Filter by Class!(
    execute
    TargetArray:Array of Object References=Make Array("[0]":Object Reference=GetAsset(InAssetData=@n9)) -> (Array:Array of Object References)
    ObjectClass:Object Class Reference=/Script/Blutility.EditorUtilityWidgetBlueprint
    FilterType:EEditorScriptingFilterType Enum
   ) -> (then, ReturnValue:Array of Editor Utility Widget Blueprint Object References)
   > SpawnAndRegisterTab(
    self=Get EditorUtilitySubsystem()
    InBlueprint="Get (a copy)"!(Array:Array of Editor Utility Widget Blueprint Object References=@n10, Dimension 1) -> (Output:Editor Utility Widget Blueprint Object Reference)
   )
   n11 = Get() -> (l_dataStruct_actorIsTransient_24_#7=~)
   Branch(execute=@n6.then_1, Condition="Not Equal Exactly (String)"(A=@n11.l_dataStruct_consoleCommand_16_#4))
   > ExecuteConsoleCommand(Command=@n11.l_dataStruct_consoleCommand_16_#4)
   > Return Node()
   n12 = Get() -> (l_dataStruct_actorIsTransient_24_#7=~)
   Branch(execute=@n6.then_2, Condition="Not Equal Exactly (String)"(A=@n12.l_dataStruct_python_17_#5))
   > ExecutePythonCommand(PythonCommand=@n12.l_dataStruct_python_17_#5)
   > Return Node()
   n13 = Break EWS Button Data Struct!(EWS_ButtonData_Struct=Get l_dataStruct() -> (l_dataStruct:EWS Button Data Struct Structure)) -> (actorPackageName_22_#6, actorIsTransient_24_#7=~)
   n14 = LoadClassAsset_Blocking(execute=@n6.then_3, AssetClass=To Soft Class Reference(SoftClassPath=MakeSoftClassPath(PathString=@n13)))
   > n15 = Cast To Actor Class!(execute:Exec, Class:Object Class Reference=@n14) -> (then:Exec, CastFailed:Exec, AsActor:Actor Class Reference) {is=title K2Node_ClassDynamicCast}
   > n16 = GetGameWorld!(execute, self:Unreal Editor Subsystem Object Reference=Get UnrealEditorSubsystem()) -> (then, ReturnValue:World Object Reference)
   > n17 = Is Valid(InputObject=@n16)
   SpawnActorFromClass!(
    execute=@n17.Is Not Valid
    ActorClass:Actor Class Reference=@n15
    Location:Vector="0, 0, 0"
    Rotation:Rotator
    bTransient:bool=@n13.actorIsTransient_24_#7
   ) -> (then, ReturnValue:Actor Object Reference)
   SpawnActor!(
    execute:Exec=@n17.Is Valid
    WorldContextObject:Object Reference=@n16
    Class:Actor Class Reference=@n15
    SpawnTransform_Location:Vector="0, 0, 0"
    SpawnTransform_Rotation:Rotator="0, 0, 0"
    SpawnTransform_Scale:Vector="1,1,1"
    CollisionHandlingOverride:ESpawnActorCollisionHandlingMethod Enum=AlwaysSpawn
    TransformScaleMethod:ESpawnActorScaleMethod Enum=SelectDefaultAtRuntime
    Owner:Actor Object Reference
    Instigator:Pawn Object Reference
   ) -> (then:Exec, ReturnValue:Actor Object Reference) {is=title K2Node_SpawnActorFromClass}
  ]}
 openAddButtonPopup {nodes [
   openAddButtonPopup() {is=title K2Node_FunctionEntry}
   > setDeleteMode(on=false)
   > n1 = SpawnAndRegisterTab(self=Get EditorUtilitySubsystem(), InBlueprint=/Game/Blueprints/Blutilities/ForTechArt/EditorWidgetShelf/SubWidgets/EWS_AddNewPopup_EWidget)
   > n2 = Cast To EWS_AddNewPopup_EWidget!(execute, Object=@n1) -> (then, CastFailed, AsEWS Add New Popup EWidget:EWS Add New Popup EWidget Object Reference) {is=title K2Node_DynamicCast}
   > n3 = Set currentAddButtonPopup(currentAddButtonPopup=@n2)
   n4 = Self-Reference()
   > n5 = setNewParent(self=@n3, newParentShelfWidget=@n4)
   SetVisibility(execute=[@n5.then, @n6.then], self=Get borderNewButtonPopup(), InVisibility=Visible)
   n6 = Set parentShelfWidget!(execute, parentShelfWidget:Editor Widget Shelf EWidget Object Reference=@n4, self:EWS Add New Popup EWidget Object Reference=@n3) -> (then, Output_Get:Editor Widget Shelf EWidget Object Reference)
  ]}
 refreshList {nodes [
   refreshList() {is=title K2Node_FunctionEntry}
   > clearItems(self:TA Generic 01 List Object Reference=Get TA_Generic01_List())
   > Set deleteModeOn(deleteModeOn:bool=false) -> (Output_Get:bool)
   > n1 = For Each Loop(Array:Array of EWS Button Data Struct Structures=Get buttons()) -> (Array Element:EWS Button Data Struct Structure)
   > addNewButton(NewItem=@n1)
  ]}
 removeButton {nodes [
   n1 = removeButton!() -> (then, button:EWS Button Widget Object Reference) {is=title K2Node_FunctionEntry}
   > n2 = getIndexOfItem!(execute, self:TA Generic 01 List Object Reference=Get TA_Generic01_List(), Content:Widget Object Reference=@n1) -> (then, index:int=0)
   > Remove Index!(execute, TargetArray:Array of EWS Button Data Struct Structures=Get buttons(), IndexToRemove:int=@n2) -> (then)
   > removeItemAtIndex!(execute, self:TA Generic 01 List Object Reference=Get TA_Generic01_List(), index:int=@n2) -> (then)
   > SaveGame_Save()
  ]}
 setDeleteMode {nodes [
   n1 = setDeleteMode!() -> (then, on:bool=~) {is=title K2Node_FunctionEntry}
   > Branch(Condition="Not Equal (Boolean)"!(A:bool=@n1, B:bool=Get deleteModeOn()) -> (ReturnValue:bool) {is=title K2Node_CallFunction})
   > Set deleteModeOn(deleteModeOn=@n1)
   > n2 = getListItems!(execute, self:TA Generic 01 List Object Reference=Get TA_Generic01_List()) -> (then, items:Array of User Widget Object References)
   > n3 = For Each Loop(Array:Array of User Widget Object References=@n2) -> (Array Element:User Widget Object Reference)
   > n4 = Cast To EWS_Button_Widget(Object=@n3)
   > setDeleteVisible!(execute, self:EWS Button Widget Object Reference=@n4, vis:bool=Get deleteModeOn()) -> (then)
  ]}
]
variables [
 buttons {type="array<EWS_ButtonData_Struct>", default []}
 currentAddButtonPopup {type=EWS_AddNewPopup_EWidget_C, default=null}
 currentButtonsNames {type="array<string>", default []}
 deleteModeOn {type=boolean}
 pythonReturn_Colors {type="array<Vector>"}
 pythonReturn_ConsoleCommands {type="array<string>", default []}
 pythonReturn_Labels {type="array<string>", default []}
 pythonReturn_Pythons {type="array<string>", default []}
 pythonReturn_WidgetPackageNames {type="array<string>", default []}
 pythonWantsUpdate {type=boolean}
]
widgets {
 info {
  parentClass=/Game/Blueprints/Blutilities/ForTechArt/Widgets/TA_Base_EditorWidget.TA_Base_EditorWidget_C
  rootWidgetClass=&CanvasPanel
  widgetCount=14
 }
 widgets [
  CanvasPanel_0 {widgetClassPath=&CanvasPanel, widget=CanvasPanel_0}
  ScaleBox_1 {
   widgetClassPath=&ScaleBox
   parent=CanvasPanel_0
   widget=ScaleBox_1
   properties {stretchDirection=DownOnly}
   slot {class=CanvasPanelSlot, properties {layoutData {alignment=<0 0.5>, anchors {maximum=<1 1>, minimum=<0 0>}, offsets=<ltrb 0 2.5 0 2.5>}}}
  }
  VerticalBox_0 {
   widgetClassPath=&VerticalBox
   parent=ScaleBox_1
   widget=VerticalBox_0
   slot {class=ScaleBoxSlot, properties {horizontalAlignment=HAlign_Fill}}
  }
  HorizontalBox {
   widgetClassPath=&HorizontalBox
   parent=VerticalBox_0
   widget=HorizontalBox
   slot {class=VerticalBoxSlot, properties {size {sizeRule=Fill, value=1}}}
  }
  Border_0 {
   widgetClassPath=&Border
   parent=HorizontalBox
   widget=Border_0
   properties {brushColor=<rgba 0 0 0 0.15>, padding=<ltrb 0 0 0 0>}
   slot {class=HorizontalBoxSlot, properties {size {sizeRule=Fill, value=1}}}
  }
  TA_Generic01_List {
   widgetClassPath=/Game/Blueprints/Blutilities/ForTechArt/Widgets/List/TA_Generic01_List
   bIsVariable=true
   parent=Border_0
   widget=TA_Generic01_List
   properties {
    enableSelection=false
    listItemClass=/Game/Blueprints/Blutilities/ForTechArt/EditorWidgetShelf/SubWidgets/EWS_Button_Widget.EWS_Button_Widget_C
    scrollOrientation=Orient_Horizontal
   }
   slot {class=BorderSlot, properties {padding=<ltrb 0 0 0 0>}}
  }
  VerticalBox_2 {widgetClassPath=&VerticalBox, parent=HorizontalBox, widget=VerticalBox_2, slot {class=HorizontalBoxSlot}}
  btnAddNewButton {
   widgetClassPath=&TA_Generic01_Button
   parent=VerticalBox_2
   widget=btnAddNewButton
   properties {
    buttonStyle_Default {
     clickedSlateSound {resourceObject=null}
     disabled {
      bIsDynamicallyLoaded=false
      drawAs=Box
      imageType=NoImage
      mirroring=NoMirror
      resourceName=None
      resourceObject=/Game/Blueprints/Blutilities/ForTechArt/ArtAssets/Textures/Button02/MI_TA_ButtonBG_02_Down
      tiling=NoTile
      imageSize=<32 32>
      margin=<ltrb 0.25 0.25 0.25 0.25>
      outlineSettings {
       bUseBrushTransparency=false
       roundingType=HalfHeightRadius
       width=0
       color {colorUseRule=UseColor_Specified, specifiedColor=<rgba 0 0 0 0>}
       cornerRadii=<0 0 0 1>
      }
      tintColor {colorUseRule=UseColor_Specified, specifiedColor=<rgba 0.3333 0.3333 0.3333 1>}
      uVRegion {bIsValid=false, max=<0 0>, min=<0 0>}
     }
     disabledForeground {colorUseRule=UseColor_Foreground, specifiedColor=<rgba 1 0 1 1>}
     hovered {
      bIsDynamicallyLoaded=false
      drawAs=Box
      imageType=NoImage
      mirroring=NoMirror
      resourceName=None
      resourceObject=/Game/Blueprints/Blutilities/ForTechArt/ArtAssets/Textures/Button02/MI_TA_ButtonBG_02_Over
      tiling=NoTile
      imageSize=<32 32>
      margin=<ltrb 0.25 0.25 0.25 0.25>
      outlineSettings {
       bUseBrushTransparency=false
       roundingType=HalfHeightRadius
       width=0
       color {colorUseRule=UseColor_Specified, specifiedColor=<rgba 0 0 0 0>}
       cornerRadii=<0 0 0 1>
      }
      tintColor {colorUseRule=UseColor_Specified, specifiedColor=<rgba 0.0998 0.8542 0.0701 1>}
      uVRegion {bIsValid=false, max=<0 0>, min=<0 0>}
     }
     hoveredForeground {colorUseRule=UseColor_Foreground, specifiedColor=<rgba 1 0 1 1>}
     hoveredSlateSound {resourceObject=null}
     normal {
      bIsDynamicallyLoaded=false
      drawAs=Box
      imageType=NoImage
      mirroring=NoMirror
      resourceName=None
      resourceObject=/Game/Blueprints/Blutilities/ForTechArt/ArtAssets/Textures/Button02/M_TA_ButtonBG_02_Up
      tiling=NoTile
      imageSize=<32 32>
      margin=<ltrb 0.25 0.25 0.25 0.25>
      outlineSettings {
       bUseBrushTransparency=false
       roundingType=HalfHeightRadius
       width=0
       color {colorUseRule=UseColor_Specified, specifiedColor=<rgba 0 0 0 0>}
       cornerRadii=<0 0 0 1>
      }
      tintColor {colorUseRule=UseColor_Specified, specifiedColor=<rgba 0.0904 0.5457 0.0008 1>}
      uVRegion {bIsValid=false, max=<0 0>, min=<0 0>}
     }
     normalForeground {colorUseRule=UseColor_Foreground, specifiedColor=<rgba 1 0 1 1>}
     normalPadding=<ltrb 0 0 0 0>
     pressed {
      bIsDynamicallyLoaded=false
      drawAs=Box
      imageType=NoImage
      mirroring=NoMirror
      resourceName=None
      resourceObject=&MI_TA_ButtonBG_02_Down
      tiling=NoTile
      imageSize=<32 32>
      margin=<ltrb 0.25 0.25 0.25 0.25>
      outlineSettings {
       bUseBrushTransparency=false
       roundingType=HalfHeightRadius
       width=0
       color {colorUseRule=UseColor_Specified, specifiedColor=<rgba 0 0 0 0>}
       cornerRadii=<0 0 0 1>
      }
      tintColor {colorUseRule=UseColor_Specified, specifiedColor=<rgba 0.0386 0.2396 0.0196 1>}
      uVRegion {bIsValid=false, max=<0 0>, min=<0 0>}
     }
     pressedForeground {colorUseRule=UseColor_Foreground, specifiedColor=<rgba 1 0 1 1>}
     pressedPadding=<ltrb 0 0 0 0>
     pressedSlateSound {resourceObject=null}
    }
    iconBrush {
     bIsDynamicallyLoaded=false
     drawAs=Image
     imageType=NoImage
     mirroring=NoMirror
     resourceName=None
     resourceObject=/Game/Athena/UI/Frontend/Parties/Art/T_UI_PlusSign_64x
     tiling=NoTile
     imageSize=<16 16>
     margin=<ltrb 0 0 0 0>
     outlineSettings {
      bUseBrushTransparency=false
      roundingType=HalfHeightRadius
      width=0
      color {colorUseRule=UseColor_Specified, specifiedColor=<rgba 0 0 0 0>}
      cornerRadii=<0 0 0 1>
     }
     tintColor {colorUseRule=UseColor_Specified, specifiedColor=<rgba 1 1 1 1>}
     uVRegion {bIsValid=false, max=<0 0>, min=<0 0>}
    }
    labelColor_Default=<rgba 1 1 1 1>
   }
   slot {class=VerticalBoxSlot, properties {horizontalAlignment=HAlign_Center, verticalAlignment=VAlign_Center, size {sizeRule=Fill, value=1}}}
  }
  btnDeleteMode {
   widgetClassPath=&TA_Generic01_Button
   parent=VerticalBox_2
   widget=btnDeleteMode
   properties {
    buttonColor_Default=<rgba 1 0.0833 0.0833 1>
    iconBrush {
     bIsDynamicallyLoaded=false
     drawAs=Image
     imageType=NoImage
     mirroring=NoMirror
     resourceName=None
     resourceObject=&Cancel_Normal
     tiling=NoTile
     imageSize=<16 16>
     margin=<ltrb 0 0 0 0>
     outlineSettings {
      bUseBrushTransparency=false
      roundingType=HalfHeightRadius
      width=0
      color {colorUseRule=UseColor_Specified, specifiedColor=<rgba 0 0 0 0>}
      cornerRadii=<0 0 0 1>
     }
     tintColor {colorUseRule=UseColor_Specified, specifiedColor=<rgba 1 1 1 1>}
     uVRegion {bIsValid=false, max=<0 0>, min=<0 0>}
    }
    labelColor_Default=<rgba 1 1 1 1>
   }
   slot {class=VerticalBoxSlot, properties {horizontalAlignment=HAlign_Center, verticalAlignment=VAlign_Center, size {sizeRule=Fill, value=1}}}
  }
  borderNewButtonPopup {
   widgetClassPath=&Border
   bIsVariable=true
   parent=CanvasPanel_0
   widget=borderNewButtonPopup
   properties {visibility=Collapsed, brushColor=<rgba 0.026 0.0175 0.0152 1>, padding=<ltrb 4 4 4 4>}
   slot {class=CanvasPanelSlot, properties {zOrder=1, layoutData {alignment=<0 0>, anchors {maximum=<1 1>, minimum=<0 0>}, offsets=<ltrb 0 0 0 0>}}}
  }
  ScaleBox_0 {
   widgetClassPath=&ScaleBox
   parent=borderNewButtonPopup
   widget=ScaleBox_0
   properties {stretchDirection=DownOnly}
   slot {class=BorderSlot, properties {padding=<ltrb 4 4 4 4>}}
  }
  VerticalBox_1 {widgetClassPath=&VerticalBox, parent=ScaleBox_0, widget=VerticalBox_1, slot {class=ScaleBoxSlot}}
  CommonTextBlock_43 {
   widgetClassPath=&CommonTextBlock
   parent=VerticalBox_1
   widget=CommonTextBlock_43
   properties {
    justification=Center
    text=Add New Button Popup Is Open
    font {
     bForceMonospaced=false
     bMaterialIsStencil=false
     fontMaterial=null
     fontObject=&NotoSans
     letterSpacing=0
     monospacedWidth=1
     size=16
     skewAmount=0
     typefaceFontName=NotoSans-Bold
     outlineSettings {
      bApplyOutlineToDropShadows=false
      bMiteredCorners=false
      bSeparateFillAlpha=false
      outlineMaterial=null
      outlineSize=0
      outlineColor=<rgba 0 0 0 1>
     }
    }
    shadowOffset=<0 0>
   }
   slot {class=VerticalBoxSlot}
  }
  TA_Generic01_Button {
   widgetClassPath=&TA_Generic01_Button
   parent=VerticalBox_1
   widget=TA_Generic01_Button
   properties {label=CANCEL, buttonColor_Default=<rgba 1 0.1108 0.0611 1>}
   slot {class=VerticalBoxSlot}
  }
 ]
}
